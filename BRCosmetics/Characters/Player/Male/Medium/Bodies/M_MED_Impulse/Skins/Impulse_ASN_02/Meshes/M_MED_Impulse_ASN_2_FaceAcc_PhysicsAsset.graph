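# Physics asset for the face accessory mesh of the Impulse ASN 02 skin, a medium male player cosmetic body.
# It holds a single capsule body on the head bone, offset and rotated from that bone.
# It defines no constraints, so it provides one collision shape rather than a jointed ragdoll setup.
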PhysicsAsset /BRCosmetics/Characters/Player/Male/Medium/Bodies/M_MED_Impulse/Skins/Impulse_ASN_02/Meshes/M_MED_Impulse_ASN_2_FaceAcc_PhysicsAsset
bodies {head [{shapeType=Capsule, length=11.1369, radius=8.7802, center=<8.2426 -2.2913 -0.2216>, rotation=<rot 0 -171.1621 -85.6438>}]}
constraints []
